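# AthenaPlayerActionAlertExtensionBase is a UserWidget Blueprint base for Athena HUD player action alert extensions.
# It declares a SetPlayerState function that takes a Player State reference, with an empty body.
# It has no widgets, variables or event graph logic, so it defines only that entry point.
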WidgetBlueprint /Game/Athena/HUD/AthenaPlayerActionAlertExtensionBase
graphs [
 EventGraph {links [], nodes []}
 SetPlayerState {nodes [SetPlayerState!() -> (then, InPlayerState:Player State Object Reference) {is=title K2Node_FunctionEntry}]}
]
variables []
widgets {info {inheritedWidgetCount=0, parentClass=&UserWidget, rootWidgetClass=null, widgetCount=0}, widgets []}
