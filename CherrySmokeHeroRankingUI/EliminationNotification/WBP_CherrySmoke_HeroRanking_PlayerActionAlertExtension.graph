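# Player action alert extension widget that shows a player's CherrySmoke Hero Ranking badge in a small fixed-size box, apparently for elimination notifications.
# SetPlayerState passes the player to the Hero Ranking view model, then calls the parent.
# UpdateRankIcon collapses the widget, async-loads the soft-referenced material instance and applies it to the lazy image.
# It then zeroes the badge's pip parameters on the dynamic material and makes the widget visible.
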WidgetBlueprint /CherrySmokeHeroRankingUI/EliminationNotification/WBP_CherrySmoke_HeroRanking_PlayerActionAlertExtension
graphs [
 EventGraph {nodes [
   n1 = GetDynamicMaterial(self=Get Image_RankIcon() -> (Image_RankIcon:Fort Lazy Image Object Reference))
   > SetScalarParameterValue(self=@n1, ParameterName=ShowMiddlePips, Value=0)
   > SetScalarParameterValue(self=@n1, ParameterName=BackgroundPipsOpacity, Value=0)
   n2 = UpdateRankIcon!() -> (OutputDelegate, then, NewRankIcon:Material Instance Soft Object Reference) {is=title K2Node_CustomEvent}
   > SetVisibility(InVisibility=Collapsed)
   > n3 = Async Load Asset(Asset=@n2.NewRankIcon)
   Is Valid(exec=@n3.Completed, InputObject=@n3)
   > n4 = Cast To MaterialInstance!(execute, Object=@n3) -> (then, CastFailed, AsMaterial Instance:Material Instance Object Reference) {is=title K2Node_DynamicCast}
   n5 = Get Image_RankIcon()
   > SetBrushFromMaterial(self=@n5, Material=@n4)
   > n6 = GetDynamicMaterial(self=@n5)
   > SetScalarParameterValue(self=@n6, ParameterName=ShowMiddlePips, Value=0)
   > SetScalarParameterValue(self=@n6, ParameterName=BackgroundPipsOpacity, Value=0)
   > SetVisibility(InVisibility=HitTestInvisible)
   n7 = Event PreConstruct()
   > Branch(Condition=@n7.IsDesignTime)
  ]}
 SetPlayerState {nodes [
   n1 = SetPlayerState!() -> (then, InPlayerState:Player State Object Reference) {is=title K2Node_FunctionEntry}
   > SetDisplayingPlayerState!(
    execute
    self:Cherry Smoke Hero Ranking VM Object Reference=Get CherrySmokeHeroRankingVM() -> (CherrySmokeHeroRankingVM:Cherry Smoke Hero Ranking VM Object Reference)
    InPlayerState:Player State Object Reference=@n1
   ) -> (then)
   > Parent: SetPlayerState!(execute, InPlayerState:Player State Object Reference=@n1) -> (then) {is=title K2Node_CallParentFunction}
  ]}
]
variables []
widgets {
 info {
  inheritedWidgetCount=0
  parentClass=/Game/Athena/HUD/AthenaPlayerActionAlertExtensionBase.AthenaPlayerActionAlertExtensionBase_C
  rootWidgetClass=&SizeBox
  widgetCount=2
 }
 widgets [
  SizeBox_88 {
   widgetClassPath=&SizeBox
   widget=SizeBox_88
   properties {
    bOverride_HeightOverride=true
    bOverride_MaxAspectRatio=true
    bOverride_MinAspectRatio=true
    bOverride_WidthOverride=true
    heightOverride=64
    widthOverride=64
   }
  }
  Image_RankIcon {
   widgetClassPath=&FortLazyImage
   bIsVariable=true
   parent=SizeBox_88
   widget=Image_RankIcon
   properties {
    brush {
     bIsDynamicallyLoaded=false
     drawAs=Image
     imageType=FullColor
     mirroring=NoMirror
     resourceName=None
     resourceObject=/CherrySmokeHeroRankingUI/Materials/BadgeInstanceMaterials/MI_UI_CherrySmokeHeroRank_FullBadge_B
     tiling=NoTile
     imageSize=<128 128>
     margin=<ltrb 0 0 0 0>
     outlineSettings {
      bUseBrushTransparency=false
      roundingType=HalfHeightRadius
      width=0
      color {colorUseRule=UseColor_Specified, specifiedColor=<rgba 0 0 0 0>}
      cornerRadii=<0 0 0 0>
     }
     tintColor {colorUseRule=UseColor_Specified, specifiedColor=<rgba 1 1 1 1>}
     uVRegion {bIsValid=false, max=<0 0>, min=<0 0>}
    }
    renderTransform {angle=0, scale=<1.5 1.5>, shear=<0 0>, translation=<0 0>}
   }
   slot {class=SizeBoxSlot, properties {padding=<ltrb 15 0 0 0>}}
  }
 ]
}
